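# Physics asset for the PressureGhoul pickaxe axe mesh, a two-handed melee weapon cosmetic.
# It holds a single body on the Melee_TwoHanded_Handle bone, built from one capsule offset along and rotated to the weapon.
# There are no constraints, so the pickaxe collides as one rigid piece with no jointed or dangling parts.
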PhysicsAsset /BRCosmetics/Weapons/FORT_Melee/Pickaxe_PressureGhoul/Meshes/PressureGhoul_Axe_PhysicsAsset
bodies {Melee_TwoHanded_Handle [{shapeType=Capsule, length=85.3304, radius=24.5594, center=<1.5935 -2.3128 41.3336>, rotation=<rot 1.6062 -90 -0.7382>}]}
constraints []
